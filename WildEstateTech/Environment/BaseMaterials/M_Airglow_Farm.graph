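# Translucent unlit material for a farm airglow haze, with two-palette color scaled by time of day.
# Its opacity chain shapes a box mask, on local position or a virtual plane, with camera-distance fade, quality-switched depth fade and a vertical edge gradient.
# A separate sum adds normal-aware ground light and ray-traced sphere falloff to that mask.
# The sine-driven breathing radius lerps between identical endpoints, so it stays constant.
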Material /WildEstateTech/Environment/BaseMaterials/M_Airglow_Farm
graph {
 properties {bAutomaticallySetUsageInEditor=false, blendMode=BLEND_Translucent, shadingModel=MSM_Unlit, usedShadingModels=Unlit}
 nodes [
  n1 = Divide(
   A=Distance(
    A=MaterialFunctionCall() {materialFunction=&WorldPositionBehindTranslucency}.Result
    B=ObjectPositionWS()
   )
   B=Add(A=NamedRerouteUsage(), B=ScalarParameter() {parameterName=Ground_Light_Radius})
  )
  n2 = SceneTexture() {sceneTextureId=PPI_WorldNormal}
  n3 = Subtract(
   A=ObjectPositionWS()
   B=MaterialFunctionCall() {materialFunction=&WorldPositionBehindTranslucency}.Result
  )
  n4 = LinearInterpolate(
   A=ScalarParameter() {defaultValue=0.5, parameterName=MinLightTop}
   Alpha=DotProduct(A=ComponentMask(None=@n2.Color) {b=true, g=true, r=true}, B=Normalize(VectorInput=@n3))
  ) {constA=1}
  n5 = Constant() {r=1}
  n6 = Multiply(A=OneMinus(None=Clamp(None=@n1)), B=ShadingPathSwitch(Default=@n4, Deferred=@n4, Forward=@n5, Mobile=@n5))
  n7 = Constant() {r=1}
  n8 = NamedRerouteUsage()
  n9 = CameraPositionWS()
  n10 = Normalize(VectorInput=Subtract(A=WorldPosition().XYZ, B=@n9) {constB=-1})
  n11 = Constant()
  n12 = Add(
   A=ObjectPositionWS()
   B=MaterialFunctionCall(R=@n11, G=@n11, B=ScalarParameter() {parameterName=sphere_height}) {materialFunction=&AppendMany}.RGB
  )
  n13 = Power(
   Base=OneMinus(None=MaterialFunctionCall!(Sphere Radius=@n7, Sphere Falloff=@n8, Unit Ray Direction=@n10, Sphere Center=@n12, Ray Origin=@n9) {materialFunction=/Game/Athena/Environments/MaterialFunctions/RayTraceSphereFalloffMinClamp}.Clamped Falloff)
   Exp=ScalarParameter() {defaultValue=1.8402, parameterName=FalloffFog}
  )
  n14 = LocalPosition()
  MaterialExpressionObjectPositionWS_0 = ObjectPositionWS()
  n15 = Convert(None=@MaterialExpressionObjectPositionWS_0) {
   nodeName=Break Float3
   convertInputs [{
     type=Vector3
     expressionInput {
      expression=&self:MaterialExpressionObjectPositionWS_0
      inputName=None
      mask=0
      maskA=0
      maskB=0
      maskG=0
      maskR=0
      outputIndex=0
     }
    }]
   convertOutputs [{type=Scalar}, {type=Scalar}, {type=Scalar}]
  }
  n16 = ObjectLocalBounds()
  n17 = Multiply(
   A=Multiply(A=ComponentMask(None=@n16.Extents) {b=true}, B=ScalarParameter() {defaultValue=1.4171, parameterName=glow_offset})
   B=MaterialFunctionCall() {materialFunction=&ObjectScale}.Scale Z
  )
  n18 = MaterialFunctionCall(
   Plane Normal Axis=Constant3Vector() {constant=<rgba 0 0 1 0>}
   Plane X Axis=Constant3Vector() {constant=<rgba 1 0 0 0>}
   Plane Center=MaterialFunctionCall(R=@n15, G=@n15, B=Add(A=@n15, B=@n17)) {materialFunction=&AppendMany}.RGB
  ) {materialFunction=&VirtualPlaneCoordinates}
  n19 = VectorParameter() {parameterName=Bounds, channelNames {a=Falloff, b="", g="", r=""}, defaultValue=<rgba 100 100 100 1>}
  n20 = Multiply(
   A=MaterialFunctionCall!(
    A=StaticSwitchParameter(True=@n14.XYZ, False=@n18.World Position - Centered) {parameterName=FarmTile}
    B
    Bounds=@n19.RGB
    Edge Falloff=@n19.RGB
   ) {materialFunction=/Engine/Functions/Engine_MaterialFunctions01/Gradient/BoxMask-3D}.Result
   B=ScalarParameter() {defaultValue=1, parameterName=top_opacity}
  )
  Add(A=Clamp(None=Add(A=@n6, B=@n13)), B=@n20)
  n21 = LinearInterpolate(
   A=VectorParameter() {parameterName=Glow_Color, defaultValue=<rgba 0.4164 2 0.0873 1>}.RGB
   B=VectorParameter() {parameterName=Glow_Color2, defaultValue=<rgba 1.7904 0 2.4062 1>}.RGB
   Alpha=ScalarParameter() {desc=Switch to secondary color palette, parameterName=Color2}
  )
  n22 = MaterialFunctionCall(
   Value_Morning=ScalarParameter() {defaultValue=0.6, group=TOD, parameterName="Value_Morning (S)"}
   Value_Day=ScalarParameter() {defaultValue=0.65, group=TOD, parameterName="Value_Day (S)"}
   Value_Evening=ScalarParameter() {defaultValue=0.65, group=TOD, parameterName="Value_Evening (S)"}
   Value_Night=ScalarParameter() {defaultValue=0.5, group=TOD, parameterName="Value_Night (S)"}
  ) {materialFunction=&Scalar_Time_of_Day}
  MaterialFunctionCall(Emissive=Multiply(A=Multiply(A=@n21, B=@n22.Result), B=ScalarParameter() {defaultValue=1, parameterName=Overall_Glow_Multi})) {materialFunction=&MF_FX_EmissiveAdjustments}
  n23 = Subtract(A=Distance(A=WorldPosition().XYZ, B=CameraPositionWS()), B=ScalarParameter() {defaultValue=15000, parameterName=CameraFadeOffset})
  n24 = ScalarParameter() {defaultValue=10000, parameterName=CameraFadeLength}
  n25 = Multiply(A=Clamp(None=Divide(A=@n23, B=@n24)), B=@n20)
  n26 = DepthFade(Opacity=@n25, FadeDistance=ScalarParameter() {defaultValue=5003.6099, parameterName=DepthFadeDistance})
  n27 = Power(
   Base=ComponentMask(None=MaterialFunctionCall() {materialFunction=&BoundingBoxBased_0-1_UVW}.RGB) {b=true}
   Exp=ScalarParameter() {defaultValue=2, parameterName=glow_Pow}
  )
  n28 = Floor(None=@n27)
  n29 = Add(A=@n27, B=ScalarParameter() {parameterName=glow_offset_Z})
  n30 = Clamp(None=Multiply(A=OneMinus(None=@n28), B=OneMinus(None=@n29)), Max=ScalarParameter() {parameterName=max_opacity_Z}) {maxDefault=0.5}
  n31 = ScalarParameter() {parameterName=s}
  Multiply(
   A=Clamp(None=QualitySwitch(Default=@n26, Low=@n25, High=@n26, Medium=@n26), Max=ScalarParameter() {defaultValue=0.13, parameterName=Max_Glow_Str})
   B=NamedRerouteDeclaration(None=DepthFade(Opacity=@n30, FadeDistance=@n31)) {name=Outer_Edge, nodeColor=<rgba 0 0.8828 1 1>}
  )
  n32 = VectorParameter() {parameterName=Brealth options, channelNames {a="", b=Speed, g=High, r=low}, defaultValue=<rgba 0.9 1 5 0>}
  n33 = MaterialFunctionCall(Speed=Divide(B=@n32.Speed)) {materialFunction=&DebugTimeSine}
  NamedRerouteDeclaration(None=Multiply(A=ScalarParameter() {defaultValue=8934.8945, parameterName=LightRadius}, B=LinearInterpolate(A=@n32.low, B=@n32.low, Alpha=@n33.Result))) {name=Light_Radius, nodeColor=<rgba 0.6875 1 0 1>}
 ]
}
