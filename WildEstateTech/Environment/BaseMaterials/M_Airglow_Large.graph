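# Translucent unlit material for a large airglow around an object, invisible up close and fading in with camera distance.
# It adds normal-weighted ground light, reconstructed from the scene behind the translucency, to a ray-traced sphere fog falloff, with depth fade except on Low quality.
# Emissive blends two palette colors, scaled by time of day.
# The sine "breath" lerps between identical low values, so the light radius stays constant.
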Material /WildEstateTech/Environment/BaseMaterials/M_Airglow_Large
graph {
 properties {bAutomaticallySetUsageInEditor=false, blendMode=BLEND_Translucent, shadingModel=MSM_Unlit, usedShadingModels=Unlit}
 nodes [
  Constant4Vector() {constant=<rgba 0.7477 0.1485 2 1>}
  n1 = VectorParameter() {parameterName=Glow_Color, defaultValue=<rgba 0.4164 2 0.0873 1>}
  n2 = VectorParameter() {parameterName=Glow_Color2, defaultValue=<rgba 1.7904 0 2.4062 1>}
  n3 = ScalarParameter() {desc=Switch to secondary color palette, parameterName=Color2}
  n4 = ScalarParameter() {defaultValue=0.6, group=TOD, parameterName="Value_Morning (S)"}
  n5 = ScalarParameter() {defaultValue=0.65, group=TOD, parameterName="Value_Day (S)"}
  n6 = ScalarParameter() {defaultValue=0.65, group=TOD, parameterName="Value_Evening (S)"}
  n7 = ScalarParameter() {defaultValue=0.5, group=TOD, parameterName="Value_Night (S)"}
  MaterialFunctionCall(Emissive=Multiply(
    A=LinearInterpolate(A=@n1.RGB, B=@n2.RGB, Alpha=@n3)
    B=MaterialFunctionCall(Value_Morning=@n4, Value_Day=@n5, Value_Evening=@n6, Value_Night=@n7) {materialFunction=&Scalar_Time_of_Day}.Result
   )) {materialFunction=&MF_FX_EmissiveAdjustments}
  n8 = ScalarParameter() {defaultValue=0.4593, parameterName=inner_cutout}
  Multiply(B=Clamp(None=Fresnel(ExponentIn=@n8)))
  n9 = VectorParameter() {parameterName=Brealth options, channelNames {a="", b=Speed, g=High, r=low}, defaultValue=<rgba 0.9 1 5 0>}
  n10 = Divide(B=@n9.Speed)
  n11 = Multiply(
   A=ScalarParameter() {defaultValue=8934.8945, parameterName=LightRadius}
   B=LinearInterpolate(
    A=@n9.low
    B=@n9.low
    Alpha=MaterialFunctionCall(Speed=@n10) {materialFunction=&DebugTimeSine}.Result
   )
  )
  Multiply(A=NamedRerouteUsage(), B=@n11)
  n12 = Distance(A=ObjectPositionWS(), B=CameraPositionWS())
  n13 = ScalarParameter() {defaultValue=15000, parameterName=CameraFadeOffset}
  n14 = Clamp(None=Divide(A=Subtract(A=@n12, B=@n13), B=ScalarParameter() {defaultValue=10000, parameterName=CameraFadeLength}))
  n15 = Divide(
   A=Distance(
    A=MaterialFunctionCall() {materialFunction=&WorldPositionBehindTranslucency}.Result
    B=ObjectPositionWS()
   )
   B=Add(A=NamedRerouteUsage(), B=ScalarParameter() {parameterName=Ground_Light_Radius})
  )
  n16 = SceneTexture() {sceneTextureId=PPI_WorldNormal}
  n17 = Subtract(
   A=ObjectPositionWS()
   B=MaterialFunctionCall() {materialFunction=&WorldPositionBehindTranslucency}.Result
  )
  n18 = LinearInterpolate(
   A=ScalarParameter() {defaultValue=0.5, parameterName=MinLightTop}
   Alpha=DotProduct(A=ComponentMask(None=@n16.Color) {b=true, g=true, r=true}, B=Normalize(VectorInput=@n17))
  ) {constA=1}
  n19 = Constant() {r=1}
  n20 = Multiply(A=OneMinus(None=Clamp(None=@n15)), B=ShadingPathSwitch(Default=@n18, Deferred=@n18, Forward=@n19, Mobile=@n19))
  n21 = Constant() {r=1}
  n22 = NamedRerouteUsage()
  n23 = CameraPositionWS()
  n24 = Normalize(VectorInput=Subtract(A=WorldPosition().XYZ, B=@n23) {constB=-1})
  n25 = Constant()
  n26 = Add(
   A=ObjectPositionWS()
   B=MaterialFunctionCall(R=@n25, G=@n25, B=ScalarParameter() {parameterName=sphere_height}) {materialFunction=&AppendMany}.RGB
  )
  n27 = Power(
   Base=OneMinus(None=MaterialFunctionCall!(Sphere Radius=@n21, Sphere Falloff=@n22, Unit Ray Direction=@n24, Sphere Center=@n26, Ray Origin=@n23) {materialFunction=/Game/Athena/Environments/MaterialFunctions/RayTraceSphereFalloffMinClamp}.Clamped Falloff)
   Exp=ScalarParameter() {defaultValue=1.8402, parameterName=FalloffFog}
  )
  n28 = Multiply(A=@n14, B=Clamp(None=Add(A=@n20, B=@n27)))
  n29 = DepthFade(Opacity=@n28, FadeDistance=ScalarParameter() {defaultValue=5003.6099, parameterName=DepthFadeDistance})
  Multiply(A=QualitySwitch(Default=@n29, Low=@n28, High=@n29, Medium=@n29), B=ScalarParameter() {defaultValue=0.13, parameterName=Max_Glow_Str})
  NamedRerouteDeclaration(None=@n11) {name=Light_Radius, nodeColor=<rgba 0.6875 1 0 1>}
  NamedRerouteDeclaration(None=@n14) {name=Player distance, nodeColor=<rgba 0.875 0 1 1>}
  ScalarParameter() {parameterName=min glow}
 ]
}
