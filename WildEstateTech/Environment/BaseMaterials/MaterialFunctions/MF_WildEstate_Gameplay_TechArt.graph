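# Wild Estate material function that gathers specific gameplay and tech-art effects for one-off assets.
# It passes material attributes through a serial chain of sub-functions: gameplay WPO, timer and console screens, BeetCalf, slush machines, bush herb, overlay, flag dissolve, gag highlights and CoveSpice character functions.
# An SFX mask input is exposed through a named reroute for use along the chain.
# Its description states that each contained function sits behind a switch defaulted to off.
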MaterialFunction /WildEstateTech/Environment/BaseMaterials/MaterialFunctions/MF_WildEstate_Gameplay_TechArt
graph {
 properties {description="Function to hold unique and specific functionality for specific \"One off\" assets. Each contained function should be behind a switch, defaulted to off"}
 nodes [
  n1 = NamedRerouteUsage()
  n2 = MaterialFunctionCall!(
   Material=MaterialFunctionCall!(Material=MaterialFunctionCall!(In=@n1) {materialFunction=&/MF_WildEstate_WPOFunctions_Gameplay}.Result) {materialFunction=/WildEstateTech/TechArt/GlowDonut/Timer/MF_WildEstate_TimerDisplay}.Material
   UVs
  ) {materialFunction=/WildEstateTech/TechArt/GlowDonut/Timer/MF_WildEstate_Console_StatusDisplay}
  n3 = NamedRerouteUsage()
  n4 = MaterialFunctionCall!(Material=MaterialFunctionCall!(Material=MaterialFunctionCall!(Material=@n2.Material, SFX=@n3) {materialFunction=/WildEstateTech/TechArt/GlowDonut/MF_Screen_MMOPowerState}.Result) {materialFunction=&/MF_BeetCalf}.Material) {materialFunction=/WildEstateTech/TechArt/GlowDonut/MF_GlowDonut_DynConsole}
  n5 = MaterialFunctionCall!(
   In=MaterialFunctionCall!(MaterialIn=MaterialFunctionCall!(MaterialIn=@n4.Result) {materialFunction=/WildEstateTech/Environment/Props/WildEstate/WildEstate_SlushMachine_A/Materials/MF_WildEstate_SlushMachine}.Material) {materialFunction=&/MF_SlushieMachine_DualFlavor}.Material
   Mask=NamedRerouteUsage()
  ) {materialFunction=&/MF_WildEstate_BushHerb_ZP}
  n6 = MaterialFunctionCall!(In=MaterialFunctionCall!(Material=MaterialFunctionCall!(Material=@n5.Result) {materialFunction=&/MF_WildEstate_ZPOverlay}.Result) {materialFunction=&/MF_WildEstate_FlagDissolve}.Result) {materialFunction=&/MF_WildEstate_GagHighlights}
  FunctionOutput(None=MaterialFunctionCall!(Material=MaterialFunctionCall!(Material=@n6.Result) {materialFunction=/WildEstateTech/TechArt/Character_Materials/Material_Functions/MF_CoveSpiceArm}.Result) {materialFunction=/WildEstateTech/TechArt/Character_Materials/Material_Functions/MF_CoveSpice_EquipMask}.Result)
  NamedRerouteDeclaration(None=FunctionInput() {bUsePreviewValueAsDefault=true, inputName=SFXMask, sortPriority=1}) {name=SFX Mask, nodeColor=<rgba 1 0 0.5625 1>}
  NamedRerouteDeclaration(None=FunctionInput() {inputName=Material, inputType=FunctionInput_MaterialAttributes}) {name=Material In, nodeColor=<rgba 1 0 0.7266 1>}
 ]
}
