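# Material function that builds one WildEstate environment surface layer as material attributes, finished through a flat-shading function.
# Base color combines texture or color, masked SFX tints and an optional UV or world-aligned overlay with distance-field line thickness.
# Roughness carries a cell-shading mask from vertex color, optionally varied by object position, while metallic and specular are zero.
# Switches add a normal map and SFX-driven emissive with Blueprint tint support.
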MaterialFunction /WildEstateTech/Environment/BaseMaterials/MaterialFunctions/MF_WildEstate_Layer
graph {nodes [
  n1 = Constant()
  n2 = NamedRerouteUsage()
  n3 = FunctionInput() {inputName=NormalTexture, inputType=FunctionInput_Texture2D, sortPriority=40}
  n4 = SetMaterialAttributes!(
   MaterialAttributes
   BaseColor=NamedRerouteUsage()
   Metallic=@n1
   Specular=@n1
   Roughness=NamedRerouteUsage()
   Normal=StaticSwitch(
    True=TextureSample(UVs=@n2, Tex=@n3) {samplerType=SAMPLERTYPE_Normal}.RGB
    False=Constant3Vector() {constant=<rgba 0 0 1 0>}
    Value=FunctionInput() {inputName=UseNormal, inputType=FunctionInput_StaticBool, sortPriority=200}
   )
  ) {attributeSetTypes [
    #1
    #2
    #3
    #4
    #5
   ]}
  n5 = SetMaterialAttributes!(MaterialAttributes=@n4, EmissiveColor=NamedRerouteUsage()) {attributeSetTypes [#6]}
  n6 = StaticSwitchParameter(
   True=SetMaterialAttributes(MaterialAttributes=@n4, EmissiveColor=NamedRerouteUsage()) {attributeSetTypes [#6]}
   False=@n4
  ) {
   desc="Emissive RGB Available if SFX RGB is activated in 0 SFX Global. Will replace layers Emissive unless Emissive RGB or Emissive Texture Add is set= true  ( adds texture sample)"
   group=A_Emissive_Texture
   parameterName=SFX Emissive Texture RGB
   sortPriority=-32
  }
  n7 = StaticSwitchParameter(True=@n6, False=@n4) {
   desc="Enable to access various emissive functions, TOD,Fresnel, Flipbook etc."
   group=A__EmissiveFilters
   parameterName=Emissive Filters
   sortPriority=-9
  }
  n8 = FunctionInput(Preview=StaticBool()) {
   bUsePreviewValueAsDefault=true
   desc="Enables controls for masked inputs (Emissive, "
   inputName=Use SFX Emissive
   inputType=FunctionInput_StaticBool
   sortPriority=-5
   previewValue=<1 0.5 0.5 1>
  }
  n9 = StaticBool()
  FunctionOutput(None=MaterialFunctionCall!(
    Material In=StaticSwitch(True=@n5, False=@n7, Value=@n8)
    UseFlatShading=FunctionInput(Preview=@n9) {bUsePreviewValueAsDefault=true, inputName=UseFlatShading, inputType=FunctionInput_StaticBool, sortPriority=999}
    IsWorldSpace
    OverrideUnlitShadinBaseColorDarkening
    OverrideUnlitShadinBaseColorEmissive
   ) {materialFunction=&/MF_WildEstate_FlatShading}.Result)
  n10 = ComponentMask(None=WorldPosition().XYZ) {b=true, r=true}
  n11 = ComponentMask(None=VectorParameter() {parameterName=DebugSize, defaultValue=<rgba -64 -64 0 1>}.RGB) {g=true, r=true}
  MaterialFunctionCall(UVs=Frac(None=Divide(A=@n10, B=@n11))) {materialFunction=&DebugScalarValues}
  n12 = StaticSwitch(
   True=StaticSwitch(
    True=NamedRerouteUsage()
    False=NamedRerouteUsage()
    Value=FunctionInput() {inputName=UseSFXTint, inputType=FunctionInput_StaticBool, sortPriority=100}
   )
   False=FunctionInput() {inputName=DiffuseColor, sortPriority=10}
   Value=FunctionInput() {inputName=UseDiffuse, inputType=FunctionInput_StaticBool, sortPriority=100}
  )
  n13 = FunctionInput() {inputName=Overlay Scale, sortPriority=302}
  n14 = FunctionInput() {inputName=Overlay_Projection_Contrast, inputType=FunctionInput_Scalar, sortPriority=103, previewValue=<1 0 0 1>}
  n15 = FunctionInput() {inputName=OverlayTexture, inputType=FunctionInput_Texture2D, sortPriority=30}
  n16 = MaterialFunctionCall(
   TextureObject=@n15
   TextureSize=Multiply(A=FunctionInput() {inputName=Overlay WS Size, inputType=FunctionInput_Scalar, sortPriority=301}, B=@n13)
   ProjectionTransitionContrast=@n14
  ) {materialFunction=&WorldAlignedTexture}
  n17 = FunctionInput(Preview=StaticBool()) {bUsePreviewValueAsDefault=true, inputName=Overlay_XYProjectionOnly, inputType=FunctionInput_StaticBool, sortPriority=102}
  n18 = TextureSample(UVs=Divide(A=NamedRerouteUsage(), B=ComponentMask(None=@n13) {g=true, r=true}), Tex=@n15) {texture=/WildEstateTech/Environment/Textures/T_WildEstate_Detail_Concrete_A_A_D}
  n19 = MaterialFunctionCall!(
   InDistanceFieldTexture=StaticSwitch(
    True=StaticSwitch(True=@n16.XY Texture, False=@n16.XY Texture, Value=@n17)
    False=@n18.R
    Value=FunctionInput() {description=Overylay use WS, inputName="Static Bool Param (True)", inputType=FunctionInput_StaticBool, sortPriority=300}
   )
   LineWidthNear
   LineWidthFar
   LineTransitionDistance
   LineTransitionOffset
   LineFadeOpacity
   IsEnabled=StaticBoolParameter() {parameterName=UseDistanceFieldOverlayTexture}
  ) {materialFunction=&/MF_WildEstate_DistanceFieldLineThickness}
  NamedRerouteDeclaration(None=StaticSwitch(
    True=Multiply(A=@n12, B=@n19.Switch_0)
    False=@n12
    Value=FunctionInput() {inputName=UseOverlay, inputType=FunctionInput_StaticBool, sortPriority=101}
   )) {name=BaseColor, nodeColor=<rgba 0 1 0.9766 1>}
  n20 = Constant() {r=0.05}
  n21 = Constant() {r=0.08}
  n22 = MaterialFunctionCall(Float3=TransformPosition(None=Constant3Vector()) {transformType=TRANSFORMPOSSOURCE_World}) {materialFunction=&BreakOutFloat3Components}
  n23 = Add(A=@n22.R, B=@n22.R)
  n24 = Frac(None=Divide(A=Add(A=@n23, B=@n22.B)) {constB=100})
  n25 = VertexColor()
  n26 = LinearInterpolate(Alpha=Multiply(A=@n24, B=LinearInterpolate(Alpha=@n25.R) {constA=0.1, constB=0.9})) {constA=0.1}
  n27 = LinearInterpolate(Alpha=@n25.R) {constA=0.1, desc=Range 0 - 0.1 is used to maks Foliage}
  NamedRerouteDeclaration(None=StaticSwitchParameter(
    True=LinearInterpolate(A=@n20, B=@n21, Alpha=@n24)
    False=StaticSwitchParameter(True=@n26, False=@n27) {group=CellShading, parameterName=UseRandomCellShade}
   ) {group=CellShading, parameterName=UseOutlineExlusionMask}) {name=RoughnessCellShading, nodeColor=<rgba 1 0.6563 0 1>}
  n28 = NamedRerouteUsage()
  NamedRerouteDeclaration(None=DotProduct(A=@n28, B=FunctionInput() {inputName=SFX Tint Channel, sortPriority=1, previewValue=<1 0 0 0>})) {name=Tint Mask, nodeColor=<rgba 0.4766 1 0 1>}
  n29 = NamedRerouteUsage()
  n30 = NamedRerouteUsage()
  n31 = LinearInterpolate(A=@n29, B=Multiply(A=@n29, B=NamedRerouteUsage()), Alpha=@n30)
  n32 = NamedRerouteUsage()
  NamedRerouteDeclaration(None=LinearInterpolate(A=@n31, B=Multiply(A=@n31, B=@n32), Alpha=@n30)) {name=Diffuse Tinted, nodeColor=<rgba 0.0312 1 0 1>}
  n33 = FunctionInput() {bUsePreviewValueAsDefault=true, inputName=SFX Diffuse Tint, sortPriority=14, previewValue=<1 1 1 1>}
  NamedRerouteDeclaration(None=@n33) {name=Diffuse Tint, nodeColor=<rgba 1 0.3984 0 1>}
  n34 = NamedRerouteUsage()
  n35 = FunctionInput() {inputName=Emissive Tint, inputType=FunctionInput_Vector4, sortPriority=13}
  n36 = NamedRerouteUsage()
  n37 = Constant() {r=1}
  n38 = StaticBoolParameter() {desc="Enable for Color tint via BP.  Used ONLY for BP color tint.", group=0_SFX BP, parameterName="SFX Layers (Tint)", sortPriority=-3}
  n39 = StaticSwitch(
   True=LinearInterpolate(
    A=@n35
    B=AppendVector(A=@n36, B=@n37)
    Alpha=ScalarParameter() {group=0_SFX BP, parameterName=Use SFX Layers as Emissive Tint, sortPriority=100}
   )
   False=@n35
   Value=@n38
  )
  n40 = ComponentMask(None=@n39) {b=true, g=true, r=true}
  NamedRerouteDeclaration(None=Multiply(A=@n34, B=@n40)) {name=Emissive Color Mult, nodeColor=<rgba 1 0 0.3281 1>}
  NamedRerouteDeclaration(None=FunctionInput(Preview=Constant3Vector() {constant=<rgba 1 1 1 0>}) {bUsePreviewValueAsDefault=true, inputName=SFX Texture Complete}) {name=SFX Tex, nodeColor=<rgba 0 0.4609 1 1>}
  n41 = NamedRerouteUsage()
  n42 = Multiply(A=@n41, B=@n34)
  n43 = NamedRerouteDeclaration(None=ComponentMask(None=@n39) {a=true}) {name=Emissive A, nodeColor=<rgba 0 1 0.2734 1>}
  NamedRerouteDeclaration(None=Multiply(A=LinearInterpolate(A=@n41, B=@n42, Alpha=@n43), B=@n40)) {name=Emissive Color SFX, nodeColor=<rgba 1 0 0.3281 1>}
  NamedRerouteDeclaration(None=TextureSample(UVs=NamedRerouteUsage(), Tex=FunctionInput() {inputName=DiffuseTexture, inputType=FunctionInput_Texture2D, sortPriority=20}) {texture=&T_WildEstate_Detail_Concrete_A_A_D}.RGB) {name=Diffuse In, nodeColor=<rgba 1 0 0.7266 1>}
  NamedRerouteDeclaration(None=DotProduct(A=@n28, B=FunctionInput() {inputName=SFX Emissive Channel, sortPriority=1, previewValue=<1 0 0 0>})) {name=Emissive Mask, nodeColor=<rgba 1 0 0.375 1>}
  NamedRerouteDeclaration(None=FunctionInput(Preview=TextureCoordinate()) {inputName=UVs, inputType=FunctionInput_Vector2, sortPriority=1}) {name=UVs, nodeColor=<rgba 0 1 0.3438 1>}
  NamedRerouteDeclaration(None=@n40) {name=Emissive Color Clear, nodeColor=<rgba 1 0 0.3281 1>}
  n44 = FunctionInput() {bUsePreviewValueAsDefault=true, inputName=Data Ramp RGB, sortPriority=14, previewValue=<1 1 1 1>}
  NamedRerouteDeclaration(None=@n44) {name=DataRampDiffuse Tint, nodeColor=<rgba 1 0.3984 0 1>}
  n45 = OneMinus(None=@n18.G)
  Saturate(None=Add(A=Round(None=@n45), B=@n25.R))
  TextureCoordinate()
 ]}
