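# Material function for Wild Estate environment materials that stylizes incoming material attributes with banded shading, controlled by static switches.
# Optional normal quantization pulls normals toward the camera vector in view-facing steps, with a simpler mobile path.
# Optional overlay quantizes vertex-normal dot sun direction (from FortniteMaterialParameters) into clamped bands, then Overlay-blends them onto base color.
# Material proxies receive unmodified input; several experimental branches, including diffuse quantization, are unconnected.
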MaterialFunction /WildEstateTech/Environment/BaseMaterials/MaterialFunctions/M_WildEstate_QantizeColor
graph {nodes [
  n1 = VertexNormalWS()
  n2 = CollectionParameter() {
   collection=&FortniteMaterialParameters
   parameterName=SunlightDirectionalVector
  }
  n3 = ComponentMask(None=@n2) {b=true, g=true, r=true}
  n4 = VectorParameter() {group=QuantizeOverlay, parameterName=QuantizeSunDirection_Mult, defaultValue=<rgba 1 1 1 1>}
  n5 = StaticSwitchParameter(True=Normalize(VectorInput=Multiply(A=@n3, B=@n4.RGB)), False=@n3) {group=QuantizeOverlay, parameterName=QuantizeOverlay_Modify_SunDirection}
  n6 = DotProduct(A=@n1, B=@n5)
  n7 = ScalarParameter() {defaultValue=0.25, group=QuantizeOverlay, parameterName=Quantize_CenterFlatten, sortPriority=1001}
  n8 = Subtract(A=Abs(None=@n6), B=@n7)
  n9 = Constant() {r=1}
  n10 = Multiply(A=Divide(A=Saturate(None=@n8), B=Subtract(A=@n9, B=@n7)), B=Sign(None=@n6))
  n11 = ScalarParameter() {defaultValue=8, group=QuantizeOverlay, parameterName=Quantize_Steps}
  n12 = Multiply(A=Divide(A=Add(A=@n10)), B=@n11)
  n13 = Max(
   A=Divide(A=Round(None=@n12), B=@n11)
   B=ScalarParameter() {defaultValue=0.25, group=QuantizeOverlay, parameterName=Quantize_MinClamp, sortPriority=1001}
  )
  n14 = ScalarParameter() {defaultValue=0.75, group=QuantizeOverlay, parameterName=Quantize_MaxClamp, sortPriority=1002}
  n15 = Saturate(None=LinearInterpolate(
    A=ScalarParameter() {group=QuantizeOverlay, parameterName=Quantize_MinRangeDarkness, sortPriority=1101}
    B=ScalarParameter() {defaultValue=0.75, group=QuantizeOverlay, parameterName=Quantize_MaxRangeBrigthness, sortPriority=1102}
    Alpha=Min(A=@n13, B=@n14)
   ))
  n16 = Add(A=DotProduct(A=@n1, B=@n5))
  Add(
   A=@n15
   B=Power(Base=Divide(A=@n16), Exp=ScalarParameter() {defaultValue=0.8546, group=QuantizeOverlay, parameterName=Quantize_RemoveSoftGradientPower})
  )
  n17 = FunctionInput() {inputType=FunctionInput_MaterialAttributes}
  n18 = GetMaterialAttributes(None=@n17) {attributeGetTypes [#1]}
  n19 = ScalarParameter() {defaultValue=0.2, group=QuantizeOverlay, parameterName=QuantizeNormalSteps_Mobile}
  n20 = Multiply(A=ComponentMask(None=@n18.Normal) {b=true, g=true, r=true}, B=@n19)
  ComponentMask(None=Divide(A=Round(None=@n20), B=@n19)) {b=true, g=true, r=true}
  Constant3Vector() {constant=<rgba 0 0 1 0>}
  n21 = CameraVectorWS()
  n22 = CrossProduct(A=@n18.Normal, B=@n21)
  CrossProduct(A=@n18.Normal, B=CrossProduct(A=@n22, B=@n21))
  Divide(A=@n9, B=@n7)
  n23 = VectorParameter() {group=QuantizeOverlay, parameterName=QuantizeNormals_Detail, defaultValue=<rgba 1000 1000 1000 1>}
  Divide(A=Round(None=Multiply(A=@n1, B=@n23.RGB)), B=@n23.RGB)
  n24 = GetMaterialAttributes(None=@n17) {attributeGetTypes [#2, #3]}
  Divide(A=@n24.BaseColor, B=Desaturation(None=@n24.BaseColor))
  Frac(None=@n20)
  n25 = DotProduct(A=@n18.Normal, B=@n21)
  n26 = Abs(None=@n25)
  n27 = ScalarParameter() {defaultValue=3, group=QuantizeOverlay, parameterName=QuantizeNormalSteps}
  n28 = Divide(A=Frac(None=Multiply(A=@n26, B=@n27)), B=@n27)
  n29 = ShadingPathSwitch(Default=LinearInterpolate(A=@n18.Normal, B=@n21, Alpha=Subtract(A=@n25, B=@n28)), Mobile=@n18.Normal)
  n30 = ShadingPathSwitch(Default=@n29, Mobile=VertexNormalWS())
  n31 = StaticSwitchParameter(
   True=SetMaterialAttributes!(
    MaterialAttributes=@n18.MaterialAttributes
    Normal=StaticSwitchParameter(True=@n30, False=@n29) {group=QuantizeOverlay, parameterName=QuantizNormalsDisableOnMobile}
   ) {attributeSetTypes [#1]}
   False=@n17
  ) {group=QuantizeOverlay, parameterName=QuantizNormals}
  n32 = GetMaterialAttributes(None=@n31) {attributeGetTypes [#2, #3]}
  n33 = SetMaterialAttributes!(
   MaterialAttributes=@n32.MaterialAttributes
   BaseColor=MaterialFunctionCall(Base=@n32.BaseColor, Blend=@n15) {materialFunction=&Blend_Overlay}.Result
  ) {attributeSetTypes [#2]}
  FunctionOutput(None=MaterialProxyReplace(Realtime=StaticSwitchParameter(True=@n33, False=@n31) {group=QuantizeOverlay, parameterName=QuantizeOveraly}, MaterialProxy=@n17))
  Length(None=@n24.BaseColor) {is=class MaterialExpressionLength}
  Normalize(VectorInput=@n22)
  PixelNormalWS()
  ScalarParameter() {defaultValue=4.4436, parameterName=TestN}
  n34 = Normalize(VectorInput=@n24.BaseColor)
  n35 = LinearInterpolate(
   A=ScalarParameter() {defaultValue=0.22, parameterName=Quantize_Diffuse_Range_Low}
   B=ScalarParameter() {defaultValue=0.528, parameterName=Quantize_Diffuse_Range_High}
   Alpha=Length(None=@n24.BaseColor)
  )
  n36 = ScalarParameter() {defaultValue=4, parameterName=Quantize_Diffuse_Steps}
  n37 = Divide(A=Round(None=Multiply(A=@n35, B=@n36)), B=@n36)
  StaticSwitchParameter(
   True=SetMaterialAttributes(MaterialAttributes=@n24.MaterialAttributes, BaseColor=Multiply(A=@n34, B=@n37)) {attributeSetTypes [#2]}
   False=@n17
  ) {group=PBWAssetStylize, parameterName=PBWAssetStylize}
  Transform() {transformSourceType=TRANSFORMSOURCE_World, transformType=TRANSFORM_Tangent}
  Transform(None=@n18.Normal)
  Transform(None=@n18.Normal)
  Transform()
  VertexNormalWS()
  VertexTangentWS()
 ]}
