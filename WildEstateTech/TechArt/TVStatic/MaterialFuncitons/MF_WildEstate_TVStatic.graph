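# Material function that adds an animated TV screen effect to incoming material attributes as emissive, gated by static switch and on/off blend.
# Screen content comes from flipbook sequences with optional sine-timed swapping and a held image, or procedural static with a blinking play-button mask.
# Scanlines, snow noise, jitter and time-of-day brightness style the picture.
# DDX-based interlacing and hue shifting run at High/Epic quality, with Fresnel dimming the glow.
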MaterialFunction /WildEstateTech/TechArt/TVStatic/MaterialFuncitons/MF_WildEstate_TVStatic
graph {nodes [
  n1 = NamedRerouteUsage()
  n2 = Divide(A=Time() {period=10}) {constB=5}
  n3 = Multiply(A=Multiply(A=Sine(None=@n2), B=Sine(None=@n2) {period=3}), B=Sine(None=@n2) {period=5})
  n4 = Constant() {r=15}
  n5 = NamedRerouteUsage()
  n6 = Multiply(
   A=MaterialFunctionCall(UVs=TextureCoordinate(), "Rotation Angle (0-1)"=ScalarParameter() {group=TVStatic, parameterName=TVStaticImage_Rotation, sortPriority=250}) {materialFunction=&CustomRotator}.Rotated Values
   B=AppendVector(
    A=ScalarParameter() {defaultValue=1, group=TVStatic, parameterName=TVStaticImage_ScaleX, sortPriority=250}
    B=ScalarParameter() {defaultValue=1, group=TVStatic, parameterName=TVStaticImage_ScaleY, sortPriority=250}
   )
  )
  n7 = VectorParameter() {group=TVStatic, parameterName=TVStaticImage__UVOffset, sortPriority=251, defaultValue=<rgba 0 0 0 1>}
  n8 = ComponentMask(None=@n7.RGB) {g=true, r=true}
  n9 = TextureSampleParameter2D(UVs=Add(A=Add(A=@n6, B=@n8), B=NamedRerouteUsage())) {
   group=TVStatic
   parameterName=TV_StaticSingleImage
   sortPriority=249
   texture=/Game/Global/Textures/Default/Blanks/T_Blank_LegalWhite_sRGBColor
  }
  n10 = NamedRerouteUsage()
  n11 = TextureSample(UVs=NamedRerouteUsage(), Tex=NamedRerouteUsage()) {
   samplerSource=SSM_Clamp_WorldGroupSettings
   samplerType=SAMPLERTYPE_Masks
   texture=/WildEstateTech/TechArt/TVStatic/Textures/T_WildEstate_Static_Mask
  }
  n12 = LinearInterpolate(
   A=Multiply(A=NamedRerouteUsage(), B=NamedRerouteUsage())
   B=LinearInterpolate(A=Multiply(A=@n9.RGB) {constB=0.6}, B=Constant3Vector() {constant=<rgba 0.3544 0.412 0.6 1>}, Alpha=Multiply(A=@n10, B=@n11.B))
   Alpha=Multiply(A=ScalarParameter() {group=TVStatic, parameterName=SingleImage_HOLD, sortPriority=250}, B=NamedRerouteUsage())
  )
  n13 = Multiply(
   A=Multiply(A=StaticSwitchParameter(True=@n5, False=@n12) {group=TVStatic, parameterName=AnimatedTVStatic, sortPriority=300}, B=NamedRerouteUsage())
   B=NamedRerouteUsage()
  )
  n14 = LinearInterpolate(
   A=@n13
   B=Add(A=@n13, B=NamedRerouteUsage())
   Alpha=ScalarParameter() {defaultValue=0.01, group=TVStatic, parameterName=SnowNoiseAmount, sortPriority=200}
  )
  n15 = Multiply(A=DDX(Value=@n14), B=ScalarParameter() {defaultValue=2, group=TVStatic, parameterName=Interlacing/blur strength, sortPriority=200}) {constB=2}
  n16 = LinearInterpolate(
   A=@n14
   B=MaterialFunctionCall(Hue Shift Percentage=Divide(A=@n3, B=@n4) {constB=30}, Texture=Add(A=@n14, B=@n15)) {materialFunction=&HueShift}.Result
   Alpha=ScalarParameter() {defaultValue=1, group=TVStatic, parameterName=FinalShifting_Amount, sortPriority=200}
  )
  n17 = MaterialFunctionCall(Emissive=QualitySwitch(Default=@n14, High=@n16, Epic=@n16)) {materialFunction=&MF_FX_EmissiveAdjustments}
  n18 = Multiply(A=@n17.Result) {constB=0.5}
  n19 = Fresnel() {baseReflectFraction=0.05, exponent=3}
  n20 = BlendMaterialAttributes(
   A=@n1
   B=SetMaterialAttributes!(
    MaterialAttributes=NamedRerouteUsage()
    BaseColor=Constant()
    EmissiveColor=LinearInterpolate(A=@n17.Result, B=@n18, Alpha=@n19)
    Roughness=Constant() {r=0.75}
   ) {attributeSetTypes [#1, #2, #3]}
   Alpha=NamedRerouteUsage()
  )
  n21 = ScalarParameter() {defaultValue=1, group=TVStatic, parameterName=TVStatic_On, sliderMax=1, sortPriority=-200}
  FunctionOutput(None=StaticSwitchParameter(True=BlendMaterialAttributes(A=@n1, B=@n20, Alpha=@n21), False=NamedRerouteUsage()) {group=Wild Estate Tech Art FX, parameterName=Use_TVStatic, sortPriority=-100})
  n22 = TextureCoordinate()
  n23 = ViewSize()
  n24 = Multiply(
   A=Floor(None=Multiply(A=@n22, B=@n23))
   B=AppendVector(
    A=ScalarParameter() {defaultValue=0.1, group=TVStatic, parameterName=Noise size X, sortPriority=200}
    B=ScalarParameter() {defaultValue=0.1, group=TVStatic, parameterName=Noise size Y, sortPriority=200}
   )
  )
  n25 = Multiply(A=Time(), B=ScalarParameter() {defaultValue=25, group=TVStatic, parameterName=Noise frame rate, sortPriority=200}) {constB=24, desc=Frame rate}
  LinearInterpolate(B=VectorNoise(World Position=AppendVector(A=@n24, B=@n25)), Alpha=NamedRerouteUsage())
  n26 = MaterialFunctionCall(UVs=TextureCoordinate(), "Rotation Angle (0-1)"=ScalarParameter() {group=TVStatic, parameterName=TVStaticRotation}) {materialFunction=&CustomRotator}
  n27 = ScalarParameter() {defaultValue=0.7, group=TVStatic, parameterName=TVStaticScaleX}
  n28 = ScalarParameter() {defaultValue=0.7, group=TVStatic, parameterName=TVStaticScaleY}
  n29 = Add(
   A=Multiply(A=@n26.Rotated Values, B=AppendVector(A=@n27, B=@n28))
   B=ComponentMask(None=VectorParameter() {group=TVStatic, parameterName=TV_UVOffset, sortPriority=12, defaultValue=<rgba 0.15 0.15 0 1>}.RGB) {g=true, r=true}
  )
  NamedRerouteDeclaration(None=@n29) {name=FinalUVs, nodeColor=<rgba 1 0.0469 0 1>}
  n30 = Sine(None=Multiply(A=Time(), B=ScalarParameter() {defaultValue=0.2, group=TVStatic, parameterName=ImageSequence_ImageChangeSpeed, sortPriority=101}))
  NamedRerouteDeclaration(None=Round(None=ConstantBiasScale(None=@n30))) {name=FLipSwitch, nodeColor=<rgba 1 0.5391 0 1>}
  NamedRerouteDeclaration(None=@n26.Rotated Values) {name=RotateUVs, nodeColor=<rgba 0.4766 1 0 1>}
  n31 = TextureCoordinate()
  n32 = ScalarParameter() {defaultValue=0.15, desc=The speed of the swipe transition, group=TVStatic, parameterName=TV_MaskSilhouette, sortPriority=49}
  n33 = Divide(B=@n32) {constB=0.25}
  n34 = Power(Base=Abs(None=ConstantBiasScale(None=@n31) {bias=-0.5, scale=2}), Exp=Add(A=Subtract(A=@n33) {constB=2}))
  n35 = OneMinus(None=Add(A=ComponentMask(None=@n34) {r=true}, B=ComponentMask(None=@n34) {g=true}))
  n36 = ChannelMaskParameter(None=FunctionInput(Preview=Saturate(None=@n35)) {bUsePreviewValueAsDefault=true, inputName=SFX Mask, sortPriority=1}) {group=TVStatic, maskChannel=Blue, parameterName=TV_SFX_ChannelMask, sortPriority=-10, defaultValue=<rgba 0 0 1 0>}
  NamedRerouteDeclaration(None=@n36.R) {name=SFX Mask, nodeColor=<rgba 0 1 0.0156 1>}
  NamedRerouteDeclaration(None=VectorParameter() {group=TVStatic, parameterName=TV_Emissive_Tint, sortPriority=50, defaultValue=<rgba 1 1 1 0>}.RGB) {name=TOD, nodeColor=<rgba 0.9453 1 0 1>}
  n37 = FunctionInput() {inputType=FunctionInput_MaterialAttributes}
  NamedRerouteDeclaration(None=GetMaterialAttributes(None=NamedRerouteDeclaration(None=@n37) {name=InMat, nodeColor=<rgba 0.3359 0 1 1>}) {attributeGetTypes [#1]}.MaterialAttributes) {name=InMatAtt, nodeColor=<rgba 0.7109 1 0 1>}
  n38 = VectorParameter() {
   desc=Image sequence params
   group=TVStatic
   parameterName=TV_ImageSequenceValues
   sortPriority=5
   channelNames {a=Overall Brightness, b=SequenceColumns, g=SequenceRows, r=SequenceSpeed}
   defaultValue=<rgba 1 2 4 1>
  }
  n39 = VectorParameter() {
   desc=Image sequence params
   group=TVStatic
   parameterName=TV2_ImageSequenceValues
   sortPriority=103
   channelNames {a=Overall Brightness, b=SequenceColumns, g=SequenceRows, r=SequenceSpeed}
   defaultValue=<rgba 1 2 4 1>
  }
  n40 = NamedRerouteUsage()
  n41 = Constant() {r=0.8}
  n42 = Constant() {r=1}
  n43 = Constant() {r=1.5}
  NamedRerouteDeclaration(None=Multiply(
    A=LinearInterpolate(A=@n38.Overall Brightness, B=@n39.Overall Brightness, Alpha=@n40)
    B=MaterialFunctionCall(Value_Morning=@n41, Value_Day=@n42, Value_Evening=@n41, Value_Night=@n43) {materialFunction=&Scalar_Time_of_Day}.Result
   )) {name=Brightness, nodeColor=<rgba 1 0.75 0 1>}
  n44 = Time()
  n45 = TextureSampleParameter2D(UVs=MaterialFunctionCall(
    "Animation  Phase (0-1)"=Multiply(A=@n44, B=@n38.SequenceSpeed)
    Number of Rows=@n38.SequenceRows
    Number of Columns=@n38.SequenceColumns
    UVs=@n29
   ) {materialFunction=&FlipBook}.UVs) {group=TVStatic, parameterName=TV_Static, sortPriority=0, texture=&T_Blank_LegalWhite_sRGBColor}
  n46 = Multiply(
   A=MaterialFunctionCall(UVs=TextureCoordinate(), "Rotation Angle (0-1)"=ScalarParameter() {group=TVStatic, parameterName=TV2StaticRotation, sortPriority=103}) {materialFunction=&CustomRotator}.Rotated Values
   B=AppendVector(
    A=ScalarParameter() {defaultValue=0.7, group=TVStatic, parameterName=TV2StaticScaleX, sortPriority=103}
    B=ScalarParameter() {defaultValue=0.7, group=TVStatic, parameterName=TV2StaticScaleY, sortPriority=103}
   )
  )
  n47 = ComponentMask(None=VectorParameter() {group=TVStatic, parameterName=TV2_UVOffset, sortPriority=103, defaultValue=<rgba 0.15 0.15 0 1>}.RGB) {g=true, r=true}
  n48 = MaterialFunctionCall(
   "Animation  Phase (0-1)"=Multiply(A=Time(), B=@n39.SequenceSpeed)
   Number of Rows=@n39.SequenceRows
   Number of Columns=@n39.SequenceColumns
   UVs=Add(A=NamedRerouteUsage(), B=Add(A=@n46, B=@n47))
  ) {materialFunction=&FlipBook}
  n49 = TextureSampleParameter2D(UVs=@n48.UVs) {group=TVStatic, parameterName=TV2_Static, sortPriority=102, texture=&T_Blank_LegalWhite_sRGBColor}
  n50 = TextureSampleParameter2D(UVs=@n48.UVs) {group=TVStatic, parameterName=TV3_Static, sortPriority=102, texture=&T_Blank_LegalWhite_sRGBColor}
  n51 = LinearInterpolate(A=@n49.RGB, B=@n50.RGB, Alpha=ScalarParameter() {group=TVStatic, parameterName=TVStatic_Bonus, sliderMax=1, sortPriority=-200})
  n52 = NamedRerouteUsage()
  NamedRerouteDeclaration(None=StaticSwitchParameter(True=LinearInterpolate(A=@n45.RGB, B=@n51, Alpha=@n52), False=@n45.RGB) {group=TVStatic, parameterName=ImageSequence_2Textures?, sortPriority=100}) {name=ImageSequence, nodeColor=<rgba 0.8984 1 0 1>}
  NamedRerouteDeclaration(None=TextureObject() {samplerType=SAMPLERTYPE_Masks, texture=&T_WildEstate_Static_Mask}) {name=TexObj, nodeColor=<rgba 0.9219 0 1 1>}
  NamedRerouteDeclaration(None=NamedRerouteUsage()) {name=SnowNoise, nodeColor=<rgba 1 0.7031 0 1>}
  n53 = Multiply(
   A=MaterialFunctionCall(UVs=TextureCoordinate(), "Rotation Angle (0-1)"=ScalarParameter() {group=TVStatic, parameterName=TVStatic_StaticRotation, sortPriority=302}) {materialFunction=&CustomRotator}.Rotated Values
   B=AppendVector(
    A=ScalarParameter() {defaultValue=2, group=TVStatic, parameterName=Static Scale X, sortPriority=301}
    B=ScalarParameter() {defaultValue=2, group=TVStatic, parameterName=Static Scale Y, sortPriority=301}
   )
  )
  n54 = MaterialFunctionCall!(
   TimeSpeed=ScalarParameter() {defaultValue=2, group=TVStatic, parameterName=Step Time Speed, sortPriority=302}
   Jump Amount Per Step=AppendVector(
    A=ScalarParameter() {defaultValue=6, group=TVStatic, parameterName=Static Offset Step Jump X, sortPriority=301}
    B=ScalarParameter() {defaultValue=4, group=TVStatic, parameterName=Static Offset Step Jump Y, sortPriority=301}
   )
   Number of Steps Per Unit of Time=AppendVector(
    A=ScalarParameter() {defaultValue=6, group=TVStatic, parameterName=Static Offset Step Smoothness X, sortPriority=301}
    B=ScalarParameter() {defaultValue=4, group=TVStatic, parameterName=Static Offset Step Smoothness Y, sortPriority=301}
   )
   Time Offset
  ) {materialFunction=/Engine/Functions/Engine_MaterialFunctions02/SteppingPannerTime}
  n55 = TextureSample(UVs=Add(A=NamedRerouteUsage(), B=Add(A=@n53, B=@n54.Result)), Tex=NamedRerouteUsage()) {
   samplerSource=SSM_Wrap_WorldGroupSettings
   samplerType=SAMPLERTYPE_Masks
   texture=&T_WildEstate_Static_Mask
  }
  n56 = Multiply(
   A=MaterialFunctionCall(UVs=TextureCoordinate(), "Rotation Angle (0-1)"=ScalarParameter() {group=TVStatic, parameterName=TVStatic_Rotation, sortPriority=302}) {materialFunction=&CustomRotator}.Rotated Values
   B=AppendVector(
    A=ScalarParameter() {defaultValue=1, group=TVStatic, parameterName=TVStatic_ScaleX, sortPriority=302}
    B=ScalarParameter() {defaultValue=1, group=TVStatic, parameterName=TVStatic_ScaleY, sortPriority=302}
   )
  )
  n57 = ComponentMask(None=VectorParameter() {group=TVStatic, parameterName=TVStatic__UVOffset, sortPriority=301, defaultValue=<rgba 0 0 0 1>}.RGB) {g=true, r=true}
  n58 = NamedRerouteDeclaration(None=Add(A=Add(A=@n56, B=@n57), B=NamedRerouteUsage())) {name=StaticUV, nodeColor=<rgba 0 0.1562 1 1>}
  n59 = TextureSample(UVs=@n58, Tex=NamedRerouteUsage()) {
   samplerSource=SSM_Clamp_WorldGroupSettings
   samplerType=SAMPLERTYPE_Masks
   texture=&T_WildEstate_Static_Mask
  }
  n60 = LinearInterpolate(
   A=LinearInterpolate(
    A=Constant3Vector() {constant=<rgba 0.0354 0.1086 0.185 1>}
    B=Constant3Vector() {constant=<rgba 0.0089 0.01 0.022 1>}
    Alpha=NamedRerouteDeclaration(None=@n55.A) {name=SimpsonsNoise, nodeColor=<rgba 0.0078 0 1 1>}
   )
   B=Constant3Vector() {constant=<rgba 0.3544 0.412 0.6 1>}
   Alpha=@n59.G
  )
  n61 = Constant3Vector() {constant=<rgba 0.0887 0.2 0.024 1>}
  n62 = TextureSample(UVs=@n58, Tex=NamedRerouteUsage()) {
   samplerSource=SSM_Clamp_WorldGroupSettings
   samplerType=SAMPLERTYPE_Masks
   texture=&T_WildEstate_Static_Mask
  }
  n63 = Multiply(A=Time(), B=ScalarParameter() {defaultValue=1, group=TVStatic, parameterName=TVStatic_PlayButtonBlinkSpeed, sortPriority=305})
  n64 = Round(None=ConstantBiasScale(None=Sine(None=@n63)))
  n65 = Multiply(A=@n62.R, B=@n64)
  NamedRerouteDeclaration(None=Multiply(
    A=LinearInterpolate(A=@n60, B=@n61, Alpha=@n65)
    B=ScalarParameter() {defaultValue=0.65, group=TVStatic, parameterName=TVStatic_Brightness, sortPriority=302}
   )) {name=Anim, nodeColor=<rgba 1 0.1172 0 1>}
  n66 = Multiply(A=Time(), B=ScalarParameter() {defaultValue=0.1, group=TVStatic, parameterName=TV2StaticNoise_JitterSpeed, sortPriority=103})
  n67 = Cosine(None=Multiply(A=@n66) {constB=4.5})
  n68 = Sine(None=Multiply(A=Add(A=@n67, B=@n66)) {constB=3})
  n69 = Sine(None=Multiply(A=@n66) {constB=4})
  n70 = Multiply(A=@n66) {constB=3.6}
  n71 = Multiply(A=Add(A=Multiply(A=@n68) {constB=0.66}, B=Multiply(A=@n69) {constB=0.5}), B=Add(A=Cosine(None=@n70), B=Sine(None=@n66)))
  NamedRerouteDeclaration(None=Saturate(None=ConstantBiasScale(None=@n71))) {name=JitterSPeed, nodeColor=<rgba 0 0.1094 1 1>}
  NamedRerouteDeclaration(None=@n64) {name=SineTime, nodeColor=<rgba 0 0.3906 1 1>}
  n72 = NamedRerouteUsage()
  n73 = ComponentMask(None=@n72) {g=true}
  n74 = Time()
  n75 = ComponentMask(None=MaterialFunctionCall!(2 Vector=AppendVector(A=@n73, B=@n74), RandomSeed) {materialFunction=&/MF_Randomize}.Random 2 Vector) {r=true}
  n76 = ComponentMask(None=ViewSize()) {r=true}
  n77 = Constant() {r=0.5}
  n78 = Saturate(None=Divide(A=Subtract(A=@n75) {constB=0.5}, B=Multiply(A=@n76, B=@n77)))
  NamedRerouteDeclaration(None=Multiply(A=Multiply(A=@n78) {constB=0.025}, B=NamedRerouteUsage())) {name=Tape, nodeColor=<rgba 1 0 0.1406 1>}
  n79 = ScalarParameter() {defaultValue=2, group=TVStatic, parameterName=ScanlineChromatic Offset}
  n80 = ComponentMask(None=Add(A=Time(), B=@n72)) {g=true}
  n81 = ComponentMask(None=ViewSize()) {g=true}
  n82 = Fmod(A=Floor(None=Multiply(A=@n80, B=@n81)), B=Constant() {r=2})
  NamedRerouteDeclaration(None=Clamp(None=Multiply(A=@n79, B=@n82), Min=ScalarParameter() {defaultValue=0.9, group=TVStatic, parameterName=ScanlineDarkValue}) {minDefault=0.9}) {name=Scanline, nodeColor=<rgba 0.7109 1 0 1>}
 ]}
