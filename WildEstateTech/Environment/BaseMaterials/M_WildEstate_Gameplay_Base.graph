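# Base gameplay material for Wild Estate environment assets, supporting instanced, skeletal and Nanite meshes.
# It blends up to four tintable layers by UV-coordinate masks, then chains effects including TV static, color quantization, flat and sun-direction shading, damage and cracks.
# A static switch replaces that chain with an anime base offering eye animation and quantized normal and sun-light overlay.
# Both paths finish through fur-silhouette and TechArt functions, using world-space normals.
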Material /WildEstateTech/Environment/BaseMaterials/M_WildEstate_Gameplay_Base
graph {
 properties {
  bAutomaticallySetUsageInEditor=false
  bTangentSpaceNormal=false
  bUseMaterialAttributes=true
  bUsedWithInstancedStaticMeshes=true
  bUsedWithMeshDeformer=true
  bUsedWithNanite=true
  bUsedWithSkeletalMesh=true
  shadingModel=MSM_FromMaterialExpression
  usedShadingModels=Default Lit | Clear Coat | Cloth
 }
 nodes [
  Composite() {subgraphName=UV Mask}
  n1 = TextureObjectParameter() {
   desc="SFX TextureGrayscale You can reuse that texture across layers and functions. "
   group=0_SFX Global
   parameterName=SFX_Grayscale
   samplerType=SAMPLERTYPE_LinearGrayscale
   sortPriority=0
   texture=/Game/Global/Base4Layers_2022/Textures/T_SFX_Grayscale
  }
  n2 = TextureObjectParameter() {
   desc="SFX UniversalTextureMask Uses Dot to select channel used for various parts of material.You can reuse that texture across layers and functions. "
   group=0_SFX Global
   parameterName=SFX_Mask
   samplerType=SAMPLERTYPE_Masks
   sortPriority=0
   texture=/Game/Global/Base4Layers_2022/Textures/T_SFX_M
  }
  NamedRerouteDeclaration(None=MaterialFunctionCall!(In=StaticSwitchParameter(True=@n1, False=@n2) {desc=Make sure size is less than 512 and lower, group=0_SFX Global, parameterName=UseSFXGrayscale}) {materialFunction=/Game/Global/Base4Layers_2022/MaterialFunctions/MF_SFXTexture}.SFX) {name=SFX V3, nodeColor=<rgba 0.6406 0 1 1>}
  n3 = MaterialFunctionCall!(CustomUV) {materialFunction=/Game/Global/Materials/Character_Parent_2023/Anime_Parent_2023/MaterialFunctions/MFC_AnimeShader_2023_Pre}
  n4 = MaterialFunctionCall!(CustomLightVector, OverrideNormal_WS, STT=@n3.STT, Normal_Map=@n3.STT, DFL=@n3.STT, FX=@n3.STT) {materialFunction=/Game/Global/Materials/Character_Parent_2023/Anime_Parent_2023/MaterialFunctions/MFC_AnimeShader_2023_Calculations}
  n5 = VertexColor()
  n6 = TextureSampleParameter2D() {
   group=01_EyeAnimation_Parameters
   parameterName=EyeMask
   samplerSource=SSM_Wrap_WorldGroupSettings
   samplerType=SAMPLERTYPE_Masks
   sortPriority=-32
   texture=&T_WhiteMask
  }
  n7 = StaticSwitchParameter(True=AppendVector(A=@n5, B=@n5), False=@n6.RGBA) {group=01_EyeAnimation_Parameters, parameterName=UseVertexColorForEyeMask, sortPriority=-33}
  n8 = TextureCoordinate() {coordinateIndex=1}
  n9 = Sign(None=Subtract(A=ComponentMask(None=@n8) {r=true}) {constB=0.0001})
  n10 = MaterialFunctionCall!(
   Color Lit=@n3.Color_Lit
   Color Shaded=@n3.Color_Lit
   Use UV2=StaticBoolParameter() {defaultValue=true, group=01_EyeAnimation_Parameters, parameterName=Use UV2 for Eye Animation, sortPriority=-30}
   Mask=StaticSwitchParameter(
    True=ChannelMaskParameter(None=@n7) {group=01_EyeAnimation_Parameters, parameterName=EyeMaskChannel, sortPriority=-32}.R
    False=Saturate(None=@n9)
   ) {group=01_EyeAnimation_Parameters, parameterName=UseEyeMask, sortPriority=-33}
  ) {materialFunction=/Game/Global/Materials/Character_Parent_2023/Anime_Parent_2023/MaterialFunctions/MF_AnimeEyeAnimation_v2}
  n11 = MaterialFunctionCall!(
   Shading_Gradient=@n4.Shading_Gradient
   Threshold_Gradient=@n4.Shading_Gradient
   Specular_Gradient=@n4.Shading_Gradient
   Sheen_Mask=@n4.Shading_Gradient
   Thickness=@n4.Shading_Gradient
   DFL_Mask=@n4.Shading_Gradient
   FX=@n4.Shading_Gradient
   Metallic_Gradient=@n4.Shading_Gradient
   RimLight_Gradient=@n4.Shading_Gradient
   NormalBuffer=@n4.Shading_Gradient
   --------
   CINE_LightSources=@n4.Shading_Gradient
   CINE_LightColor=@n4.Shading_Gradient
   ---------
   Color_Lit=@n10.Lit
   Color_Shaded=@n10.Lit
   SRM=@n3.SRM
   Emissive=@n3.SRM
   DFL_Color=@n3.SRM
   Specular_Color=@n3.SRM
   RimLight_Color=@n3.SRM
  ) {materialFunction=/Game/Global/Materials/Character_Parent_2023/Anime_Parent_2023/MaterialFunctions/MFC_AnimeShader_2023_Post}
  n12 = GetMaterialAttributes(None=@n11.Result) {attributeGetTypes [#1]}
  n13 = CameraVectorWS()
  n14 = DotProduct(A=@n12.Normal, B=@n13)
  n15 = Abs(None=@n14)
  n16 = ScalarParameter() {defaultValue=3, group=QuantizeOverlay, parameterName=QuantizeNormalSteps}
  n17 = Divide(A=Frac(None=Multiply(A=@n15, B=@n16)), B=@n16)
  n18 = ShadingPathSwitch(Default=LinearInterpolate(A=@n12.Normal, B=@n13, Alpha=Subtract(A=@n14, B=@n17)), Mobile=@n12.Normal)
  n19 = ShadingPathSwitch(Default=@n18, Mobile=VertexNormalWS())
  n20 = StaticSwitchParameter(
   True=SetMaterialAttributes!(
    MaterialAttributes=@n12.MaterialAttributes
    Normal=StaticSwitchParameter(True=@n19, False=@n18) {group=QuantizeOverlay, parameterName=QuantizNormalsDisableOnMobile}
   ) {attributeSetTypes [#1]}
   False=@n11.Result
  ) {group=QuantizeOverlay, parameterName=QuantizNormals}
  n21 = GetMaterialAttributes(None=@n20) {attributeGetTypes [#2, #3]}
  n22 = ScalarParameter() {group=QuantizeOverlay, parameterName=Quantize_MinRangeDarkness, sortPriority=1101}
  n23 = ScalarParameter() {defaultValue=0.75, group=QuantizeOverlay, parameterName=Quantize_MaxRangeBrigthness, sortPriority=1102}
  n24 = CollectionParameter() {
   collection=&FortniteMaterialParameters
   parameterName=SunlightDirectionalVector
  }
  n25 = ComponentMask(None=@n24) {b=true, g=true, r=true}
  n26 = Multiply(A=@n25, B=VectorParameter() {group=QuantizeOverlay, parameterName=QuantizeSunDirection_Mult, defaultValue=<rgba 1 1 1 1>}.RGB)
  n27 = DotProduct(
   A=VertexNormalWS()
   B=StaticSwitchParameter(True=Normalize(VectorInput=@n26), False=@n25) {group=QuantizeOverlay, parameterName=QuantizeOverlay_Modify_SunDirection}
  )
  n28 = ScalarParameter() {defaultValue=0.25, group=QuantizeOverlay, parameterName=Quantize_CenterFlatten, sortPriority=1001}
  n29 = Saturate(None=Subtract(A=Abs(None=@n27), B=@n28))
  n30 = Subtract(A=Constant() {r=1}, B=@n28)
  n31 = Add(A=Multiply(A=Divide(A=@n29, B=@n30), B=Sign(None=@n27)))
  n32 = ScalarParameter() {defaultValue=8, group=QuantizeOverlay, parameterName=Quantize_Steps}
  n33 = Round(None=Multiply(A=Divide(A=@n31), B=@n32))
  n34 = Min(
   A=Max(A=Divide(A=@n33, B=@n32), B=ScalarParameter() {defaultValue=0.25, group=QuantizeOverlay, parameterName=Quantize_MinClamp, sortPriority=1001})
   B=ScalarParameter() {defaultValue=0.75, group=QuantizeOverlay, parameterName=Quantize_MaxClamp, sortPriority=1002}
  )
  n35 = MaterialFunctionCall(Base=@n21.Emissive Color, Blend=Saturate(None=LinearInterpolate(A=@n22, B=@n23, Alpha=@n34))) {materialFunction=&Blend_Overlay}
  NamedRerouteDeclaration(None=StaticSwitchParameter(
    True=SetMaterialAttributes!(MaterialAttributes=@n21.MaterialAttributes, Emissive Color=@n35.Result) {attributeSetTypes [#2]}
    False=@n20
   ) {group=QuantizeOverlay, parameterName=QuantizeOveraly}) {name=AnimeBase, nodeColor=<rgba 0 0.5312 1 1>}
  n36 = VectorParameter() {
   desc=Use R or G or B to select your SFX Mask Channel
   group=1_Layer
   parameterName=SFX_Tint_Ch_1
   sortPriority=4
   defaultValue=<rgba 1 0 0 0>
  }
  n37 = VectorParameter() {group=1_Layer, parameterName="L1 Tint (Expand)", defaultValue=<rgba 0.6105 0.3613 0.2159 1>}
  n38 = VectorParameter() {
   desc="Emissive color, by default mult by Base color. You can control that blending via Alpha channel of this input. Uses SFX texture as the Mask"
   group=1_Layer SFX
   parameterName="SFX_Emissive_Tint(Expand)_1"
   sortPriority=-6
   channelNames {a=Base color to Emissive, b="", g="", r=""}
   defaultValue=<rgba 0 0 0 1>
  }
  n39 = ScalarParameter() {
   bUseCustomPrimitiveData=true
   desc="For now Scalar, this is possible alternative to Texture Swatch. Uses single ramp texture shared between all 4 layers and can have indepent row selected per layer. Enabled globally via function. Will disable use of color swatch if enabled. Diffuse tint still can be used if needed via separate mask "
   group=Y_CPD
   parameterName=CPD_PackedHSV_1
   primitiveDataIndex=30
   sliderMin=1
  }
  n40 = TextureObjectParameter() {
   desc=Not Really Specular related
   group=1_Layer_Overlay
   parameterName=SpecularMasks
   sortPriority=1000
   texture=/WildEstateTech/Environment/Textures/T_WildEstate_Detail_Concrete_A_A_D
   channelNames {a="", b="", g="", r=OverlayTexture}
  }
  n41 = TextureObjectParameter() {group=1_Layer, parameterName=Normals, samplerType=SAMPLERTYPE_Normal, sortPriority=0, texture=/Engine/EngineMaterials/FlatNormal}
  n42 = VectorParameter() {group=1_Layer_Overlay, parameterName=L1_OverlayScale, sortPriority=58, defaultValue=<rgba 1 1 1 1>}
  NamedRerouteDeclaration(None=MaterialFunctionCall!(
    Use SFX Emissive=StaticBoolParameter() {desc=Enables SFX, group=1_Layer SFX, parameterName=SFX Emissive, sortPriority=-8}
    SFX Texture Complete=NamedRerouteUsage()
    SFX Tint Channel=@n36.RGB
    UVs=MaterialFunctionCall!(L3 UV2, L2 UV2, L4 UV2) {materialFunction=/Game/Global/Base4Layers_2022/MaterialFunctions/MF_UVsSet_New}.UVs Layer 1
    SFX Emissive Channel=VectorParameter() {desc=Use R or G or B to select your SFX Mask Channel, group=1_Layer SFX, parameterName=SFX_Emissive_Ch_1, sortPriority=-7}.RGB
    DiffuseColor=@n37.RGB
    Emissive Tint=AppendVector(A=@n38, B=@n38)
    SFX Diffuse Tint=@n37.RGB
    Data Ramp RGB=MaterialFunctionCall!(In=@n39) {materialFunction=/Game/Global/Base4Layers_2022/MaterialFunctions/MF_ColorTint_Unpack}.Result
    DiffuseTexture=TextureObjectParameter() {group=1_Layer, parameterName=Diffuse, sortPriority=0, texture=/Game/Global/Textures/Default/Blanks/T_MidgreysRGB}
    OverlayTexture=@n40
    NormalTexture=@n41
    UseSFXTint=StaticBoolParameter() {group=1_Layer, parameterName=L1 Use SFX Tint, sortPriority=1}
    UseDiffuse=StaticBoolParameter() {group=1_Layer, parameterName=L1 Use Diffuse}
    UseOverlay=StaticBoolParameter() {group=1_Layer_Overlay, parameterName=L1 Use OverlayTexture, sortPriority=3}
    Overlay_XYProjectionOnly=StaticBoolParameter() {defaultValue=true, group=1_Layer_Overlay, parameterName=L1 Overlay XY Only, sortPriority=5}
    Overlay_Projection_Contrast=ScalarParameter() {defaultValue=1, group=1_Layer_Overlay, parameterName=L1_Overlay Porjection Contrast}
    UseNormal=StaticBoolParameter() {group=1_Layer, parameterName=L1 Use Normal}
    "Static Bool Param (True)"=StaticBoolParameter() {defaultValue=true, group=1_Layer_Overlay, parameterName=L1 Overlay WS projection, sortPriority=5}
    Overlay WS Size=ScalarParameter() {defaultValue=256, group=1_Layer_Overlay, parameterName=L1_OverlaySize}
    Overlay Scale=@n42.RGB
    UseFlatShading=StaticBoolParameter() {group=FlatShading, parameterName=UseFlatShading_Layer1, sortPriority=1}
   ) {materialFunction=/WildEstateTech/Environment/BaseMaterials/MaterialFunctions/MF_WildEstate_Layer}.Result) {name=L1, nodeColor=<rgba 0 1 0.3438 1>}
  n43 = VectorParameter() {
   desc=Use R or G or B to select your SFX Mask Channel
   group=2_Layer
   parameterName=SFX_Tint_Ch_2
   sortPriority=4
   defaultValue=<rgba 1 0 0 0>
  }
  n44 = StaticBoolParameter() {
   desc=Enable to use UV2 for mapping. You may need to switch masks for ohter layers To UV3
   group=2_layer UV
   parameterName=Use UV2 for Layer 2
  }
  n45 = VectorParameter() {group=2_Layer, parameterName="L2 Tint (Expand)", defaultValue=<rgba 0.6105 0.3613 0.2159 1>}
  n46 = VectorParameter() {
   desc="Emissive color, by default mult by Base color. You can control that blending via Alpha channel of this input. Uses SFX texture as the Mask"
   group=2_Layer SFX
   parameterName="SFX_Emissive_Tint(Expand)_2"
   sortPriority=-6
   channelNames {a=Base color to Emissive, b="", g="", r=""}
   defaultValue=<rgba 0 0 0 1>
  }
  n47 = ScalarParameter() {
   bUseCustomPrimitiveData=true
   desc="For now Scalar, this is possible alternative to Texture Swatch. Uses single ramp texture shared between all 4 layers and can have indepent row selected per layer. Enabled globally via function. Will disable use of color swatch if enabled. Diffuse tint still can be used if needed via separate mask "
   group=Y_CPD
   parameterName=CPD_PackedHSV_2
   primitiveDataIndex=31
   sliderMin=1
  }
  n48 = TextureObjectParameter() {group=2_Layer, parameterName=Diffuse_Texture_2, sortPriority=0, texture=&T_MidgreysRGB}
  n49 = TextureObjectParameter() {
   desc="R = Specular\r\nG = Metal\r\nB = Roughness"
   group=2_Layer_Overlay
   parameterName=SpecularMasks_2
   samplerType=SAMPLERTYPE_Masks
   sortPriority=0
   texture=&T_Fortnite_Default_S
  }
  n50 = TextureObjectParameter() {group=2_Layer, parameterName=Normals_Texture_2, samplerType=SAMPLERTYPE_Normal, sortPriority=0, texture=&FlatNormal}
  n51 = VectorParameter() {group=2_Layer_Overlay, parameterName=L2_OverlayScale, sortPriority=58, defaultValue=<rgba 1 1 1 1>}
  NamedRerouteDeclaration(None=MaterialFunctionCall(
    Use SFX Emissive=StaticBoolParameter() {desc=Enables SFX, group=2_Layer SFX, parameterName=SFX Emissive 2, sortPriority=-8}
    SFX Texture Complete=NamedRerouteUsage()
    SFX Tint Channel=@n43.RGB
    UVs=MaterialFunctionCall(L2 UV2=@n44) {materialFunction=&MF_UVsSet_New}.UVs Layer 2
    SFX Emissive Channel=VectorParameter() {desc=Use R or G or B to select your SFX Mask Channel, group=2_Layer SFX, parameterName=SFX_Emissive_Ch_2, sortPriority=-7}.RGB
    DiffuseColor=@n45.RGB
    Emissive Tint=AppendVector(A=@n46, B=@n46)
    SFX Diffuse Tint=@n45.RGB
    Data Ramp RGB=MaterialFunctionCall(In=@n47) {materialFunction=&MF_ColorTint_Unpack}.Result
    DiffuseTexture=@n48
    OverlayTexture=@n49
    NormalTexture=@n50
    UseSFXTint=StaticBoolParameter() {group=2_Layer, parameterName=L2 Use SFX Tint, sortPriority=1}
    UseDiffuse=StaticBoolParameter() {group=2_Layer, parameterName=L2 Use Diffuse}
    UseOverlay=StaticBoolParameter() {group=2_Layer_Overlay, parameterName=L2 Use OverlayTexture, sortPriority=3}
    Overlay_XYProjectionOnly=StaticBoolParameter() {defaultValue=true, group=2_Layer_Overlay, parameterName=L2 Overlay XY Only, sortPriority=5}
    Overlay_Projection_Contrast=ScalarParameter() {defaultValue=1, group=2_Layer_Overlay, parameterName=L2_Overlay Porjection Contrast}
    UseNormal=StaticBoolParameter() {group=2_Layer, parameterName=L2 Use Normal}
    "Static Bool Param (True)"=StaticBoolParameter() {defaultValue=true, group=2_Layer_Overlay, parameterName=L2Overlay WS projection, sortPriority=5}
    Overlay WS Size=ScalarParameter() {defaultValue=256, group=2_Layer_Overlay, parameterName=L2_OverlaySize}
    Overlay Scale=@n51.RGB
    UseFlatShading=StaticBoolParameter() {group=FlatShading, parameterName=UseFlatShading_Layer2, sortPriority=1}
   ) {materialFunction=&MF_WildEstate_Layer}.Result) {name=L2, nodeColor=<rgba 0 1 0.3438 1>}
  n52 = TextureObjectParameter() {
   desc="SFX RGB - Graffiti, emissive RGB "
   group=0_SFX Global
   parameterName=SFX_RGB
   sortPriority=1
   texture=/Game/Global/Base4Layers_2022/Textures/T-GraffitiSample_D
  }
  NamedRerouteDeclaration(None=@n52) {name=SFX Texture RGB, nodeColor=<rgba 0 0.5312 1 1>}
  n53 = VectorParameter() {
   desc=Use R or G or B to select your SFX Mask Channel
   group=3_Layer
   parameterName=SFX_Tint_Ch_3
   sortPriority=4
   defaultValue=<rgba 1 0 0 0>
  }
  n54 = StaticBoolParameter() {group=3_layer UV, parameterName=Use UV2 for Layer 3}
  n55 = VectorParameter() {group=3_Layer, parameterName="L3 Tint (Expand)", defaultValue=<rgba 0.6105 0.3613 0.2159 1>}
  n56 = VectorParameter() {
   desc="Emissive color, by default mult by Base color. You can control that blending via Alpha channel of this input. Uses SFX texture as the Mask"
   group=3_Layer SFX
   parameterName="SFX_Emissive_Tint(Expand)_3"
   sortPriority=-6
   channelNames {a=Base color to Emissive, b="", g="", r=""}
   defaultValue=<rgba 0 0 0 1>
  }
  n57 = ScalarParameter() {
   bUseCustomPrimitiveData=true
   desc="For now Scalar, this is possible alternative to Texture Swatch. Uses single ramp texture shared between all 4 layers and can have indepent row selected per layer. Enabled globally via function. Will disable use of color swatch if enabled. Diffuse tint still can be used if needed via separate mask "
   group=Y_CPD
   parameterName=CPD_PackedHSV_3
   primitiveDataIndex=35
   sliderMin=1
  }
  n58 = TextureObjectParameter() {group=3_Layer, parameterName=Diffuse_Texture_3, sortPriority=0, texture=&T_MidgreysRGB}
  n59 = TextureObjectParameter() {
   desc="R = Specular\r\nG = Metal\r\nB = Roughness"
   group=3_Layer_Overlay
   parameterName=SpecularMasks_3
   samplerType=SAMPLERTYPE_Masks
   sortPriority=0
   texture=&T_Fortnite_Default_S
  }
  n60 = TextureObjectParameter() {group=3_Layer, parameterName=Normals_Texture_3, samplerType=SAMPLERTYPE_Normal, sortPriority=0, texture=&FlatNormal}
  n61 = VectorParameter() {group=3_Layer_Overlay, parameterName=L3_OverlayScale, sortPriority=58, defaultValue=<rgba 1 1 1 1>}
  NamedRerouteDeclaration(None=MaterialFunctionCall(
    Use SFX Emissive=StaticBoolParameter() {desc=Enables SFX, group=3_Layer SFX, parameterName=SFX Emissive 3, sortPriority=-8}
    SFX Texture Complete=NamedRerouteUsage()
    SFX Tint Channel=@n53.RGB
    UVs=MaterialFunctionCall(L3 UV2=@n54) {materialFunction=&MF_UVsSet_New}.UVs Layer 3
    SFX Emissive Channel=VectorParameter() {group=3_Layer SFX, parameterName=SFX_Emissive_Ch_3, sortPriority=-7}.RGB
    DiffuseColor=@n55.RGB
    Emissive Tint=AppendVector(A=@n56, B=@n56)
    SFX Diffuse Tint=@n55.RGB
    Data Ramp RGB=MaterialFunctionCall(In=@n57) {materialFunction=&MF_ColorTint_Unpack}.Result
    DiffuseTexture=@n58
    OverlayTexture=@n59
    NormalTexture=@n60
    UseSFXTint=StaticBoolParameter() {group=1_Layer, parameterName=L3 Use SFX Tint, sortPriority=1}
    UseDiffuse=StaticBoolParameter() {group=3_Layer, parameterName=L3 Use Diffuse}
    UseOverlay=StaticBoolParameter() {group=3_Layer_Overlay, parameterName=L3 Use OverlayTexture, sortPriority=3}
    Overlay_XYProjectionOnly=StaticBoolParameter() {defaultValue=true, group=3_Layer_Overlay, parameterName=L3 Overlay XY Only, sortPriority=5}
    Overlay_Projection_Contrast=ScalarParameter() {defaultValue=1, group=3_Layer_Overlay, parameterName=L3_Overlay Porjection Contrast}
    UseNormal=StaticBoolParameter() {group=3_Layer, parameterName=L3 Use Normal}
    "Static Bool Param (True)"=StaticBoolParameter() {defaultValue=true, group=3_Layer_Overlay, parameterName=L3 Overlay WS projection, sortPriority=5}
    Overlay WS Size=ScalarParameter() {defaultValue=256, group=3_Layer_Overlay, parameterName=L3_OverlaySize}
    Overlay Scale=@n61.RGB
    UseFlatShading=StaticBoolParameter() {group=FlatShading, parameterName=UseFlatShading_Layer3, sortPriority=1}
   ) {materialFunction=&MF_WildEstate_Layer}.Result) {name=L3, nodeColor=<rgba 0 1 0.3438 1>}
  n62 = VectorParameter() {
   desc=Use R or G or B to select your SFX Mask Channel
   group=4_Layer
   parameterName=SFX_Tint_Ch_4
   sortPriority=4
   defaultValue=<rgba 1 0 0 0>
  }
  n63 = StaticBoolParameter() {group=4_layer UV, parameterName=Use UV2 for Layer 4}
  n64 = VectorParameter() {group=4_Layer, parameterName="L4 Tint (Expand)", defaultValue=<rgba 0.6105 0.3613 0.2159 1>}
  n65 = VectorParameter() {
   desc="Emissive color, by default mult by Base color. You can control that blending via Alpha channel of this input. Uses SFX texture as the Mask"
   group=4_Layer SFX
   parameterName="SFX_Emissive_Tint(Expand)_4"
   sortPriority=-6
   channelNames {a=Base color to Emissive, b="", g="", r=""}
   defaultValue=<rgba 0 0 0 1>
  }
  n66 = TextureObjectParameter() {group=4_Layer, parameterName=Diffuse_Texture_4, sortPriority=0, texture=&T_MidgreysRGB}
  n67 = TextureObjectParameter() {
   desc="R = Specular\r\nG = Metal\r\nB = Roughness"
   group=4_Layer_Overlay
   parameterName=SpecularMasks_4
   samplerType=SAMPLERTYPE_Masks
   sortPriority=0
   texture=&T_Fortnite_Default_S
  }
  n68 = TextureObjectParameter() {group=4_Layer, parameterName=Normals_Texture_4, samplerType=SAMPLERTYPE_Normal, sortPriority=0, texture=&FlatNormal}
  n69 = VectorParameter() {group=4_Layer_Overlay, parameterName=L4_OverlayScale, sortPriority=58, defaultValue=<rgba 1 1 1 1>}
  NamedRerouteDeclaration(None=MaterialFunctionCall(
    Use SFX Emissive=StaticBoolParameter() {desc=Enables SFX, group=4_Layer SFX, parameterName=SFX Emissive 4, sortPriority=-8}
    SFX Texture Complete=NamedRerouteUsage()
    SFX Tint Channel=@n62.RGB
    UVs=MaterialFunctionCall(L4 UV2=@n63) {materialFunction=&MF_UVsSet_New}.UVs Layer 4
    SFX Emissive Channel=VectorParameter() {group=4_Layer SFX, parameterName=SFX_Emissive_Ch_4, sortPriority=-7}.RGB
    DiffuseColor=@n64.RGB
    Emissive Tint=AppendVector(A=@n65, B=@n65)
    SFX Diffuse Tint=@n64.RGB
    DiffuseTexture=@n66
    OverlayTexture=@n67
    NormalTexture=@n68
    UseSFXTint=StaticBoolParameter() {group=1_Layer, parameterName=L4 Use SFX Tint, sortPriority=1}
    UseDiffuse=StaticBoolParameter() {group=4_Layer, parameterName=L4 Use Diffuse}
    UseOverlay=StaticBoolParameter() {group=4_Layer_Overlay, parameterName=L4 Use OverlayTexture, sortPriority=3}
    Overlay_XYProjectionOnly=StaticBoolParameter() {defaultValue=true, group=4_Layer_Overlay, parameterName=L4 Overlay XY Only, sortPriority=5}
    Overlay_Projection_Contrast=ScalarParameter() {defaultValue=1, group=4_Layer_Overlay, parameterName=L4_Overlay Porjection Contrast}
    UseNormal=StaticBoolParameter() {group=4_Layer, parameterName=L4 Use Normal}
    "Static Bool Param (True)"=StaticBoolParameter() {defaultValue=true, group=4_Layer_Overlay, parameterName=L4 Overlay WS projection, sortPriority=5}
    Overlay WS Size=ScalarParameter() {defaultValue=256, group=4_Layer_Overlay, parameterName=L4_OverlaySize}
    Overlay Scale=@n69.RGB
    UseFlatShading=StaticBoolParameter() {group=FlatShading, parameterName=UseFlatShading_Layer4, sortPriority=1}
   ) {materialFunction=&MF_WildEstate_Layer}.Result) {name=L4, nodeColor=<rgba 0 1 0.3438 1>}
  PinBase()
  n70 = ComponentMask(None=TextureCoordinate() {coordinateIndex=2}) {r=true}
  n71 = ComponentMask(None=TextureCoordinate() {coordinateIndex=1}) {r=true}
  n72 = Constant()
  n73 = Constant() {r=1}
  n74 = If(
   A=Constant() {r=1}
   B=StaticSwitchParameter(True=@n70, False=@n71) {group=2_Layer, parameterName=Use UV2 For Layer 2 Mask, sortPriority=-2}
   "A > B"=@n72
   "A < B"=@n73
  ) {desc=0}
  n75 = If(
   A=Constant() {r=2}
   B=StaticSwitchParameter(True=@n70, False=@n71) {group=3_Layer, parameterName=Use UV2 For Layer 3 Mask, sortPriority=-2}
   "A > B"=@n72
   "A < B"=@n73
  )
  n76 = If(
   A=Constant() {r=3}
   B=StaticSwitchParameter(True=@n70, False=@n71) {group=4_Layer, parameterName=Use UV2 For Layer 4 Mask, sortPriority=-2}
   "A > B"=@n72
   "A < B"=@n73
  )
  PinBase!(Output=@n74, Output2=@n75, Output3=@n76) {reroutePins [
    Output {expression=&self:MaterialExpressionReroute_72}
    Output2 {expression=&self:MaterialExpressionReroute_73}
    Output3 {expression=&self:MaterialExpressionReroute_74}
   ]}
  n77 = NamedRerouteUsage()
  MaterialExpressionReroute_72 = Reroute(None=@n74)
  n78 = BlendMaterialAttributes(A=NamedRerouteUsage(), B=NamedRerouteUsage(), Alpha=@MaterialExpressionReroute_72.Output)
  MaterialExpressionReroute_73 = Reroute(None=@n75)
  n79 = BlendMaterialAttributes(A=@n78, B=NamedRerouteUsage(), Alpha=@MaterialExpressionReroute_73.Output2)
  n80 = NamedRerouteUsage()
  MaterialExpressionReroute_74 = Reroute(None=@n76)
  n81 = StaticSwitch(True=@n78, False=NamedRerouteUsage(), Value=StaticBoolParameter() {group=2_Layer, parameterName=Use 2 Layers, sortPriority=-10})
  n82 = StaticBoolParameter() {group=3_Layer, parameterName=Use 3 Layers, sortPriority=-10}
  n83 = GetMaterialAttributes(None=StaticSwitch(
    True=BlendMaterialAttributes(A=@n79, B=@n80, Alpha=@MaterialExpressionReroute_74.Output3)
    False=StaticSwitch(True=@n79, False=@n81, Value=@n82)
    Value=StaticBoolParameter() {group=4_Layer, parameterName=Use 4 Layers, sortPriority=-10}
   )) {attributeGetTypes [#1]}
  n84 = MaterialFunctionCall!(
   In=SetMaterialAttributes!(MaterialAttributes=@n83.MaterialAttributes, Normal=Transform(None=@n83.Normal)) {attributeSetTypes [#1]}
   SFX Mask=NamedRerouteUsage()
  ) {materialFunction=/WildEstateTech/TechArt/TVStatic/MaterialFuncitons/MF_WildEstate_TVStatic}
  n85 = NamedRerouteUsage()
  n86 = MaterialFunctionCall!(In=MaterialFunctionCall!(In=MaterialFunctionCall!(SFX=@n77, Material In=@n84.Result, SFX Textures RGB=@n85) {materialFunction=/Game/Global/Base4Layers_2022/MaterialFunctions/MF_EmissiveControls_2}.Result) {materialFunction=/Game/ArtTools/Tech_Art/DirectFX/Assets/MaterialFunctions/MF_DirectFX}.Result) {materialFunction=/Game/ArtTools/Tech_Art/Veneer/MaterialFunctions/MF_Veneer}
  n87 = MaterialFunctionCall!(In=MaterialFunctionCall!(
    Material In=MaterialFunctionCall!(In=@n86.Result) {materialFunction=/WildEstateTech/Environment/BaseMaterials/MaterialFunctions/M_WildEstate_QantizeColor}.Result
    UseFlatShading=StaticBoolParameter() {group=FlatShading, parameterName=UseFlatShading_All, sortPriority=1}
    IsWorldSpace=StaticBool() {value=true}
    OverrideUnlitShadinBaseColorDarkening
    OverrideUnlitShadinBaseColorEmissive
   ) {materialFunction=/WildEstateTech/Environment/BaseMaterials/MaterialFunctions/MF_WildEstate_FlatShading}.Result) {materialFunction=/WildEstateTech/Environment/BaseMaterials/MaterialFunctions/M_WildEstate_SunDirectionShading_Gameplay}
  n88 = GetMaterialAttributes(None=@n87.Result) {attributeGetTypes [#1]}
  n89 = TwoSidedSign()
  n90 = StaticSwitchParameter(
   True=SetMaterialAttributes(MaterialAttributes=@n88.MaterialAttributes, Normal=Multiply(A=@n88.Normal, B=@n89)) {attributeSetTypes [#1]}
   False=@n87.Result
  ) {group=Shading, parameterName=FlipBackFaceNormals}
  n91 = StaticSwitchParameter(
   True=SetMaterialAttributes!(MaterialAttributes=@n90, Shading Model=ShadingModel() {shadingModel=MSM_ClearCoat}, Clear Coat or Cloth or Custom Data 0=Constant()) {attributeSetTypes [#4, #5]}
   False=SetMaterialAttributes!(MaterialAttributes=@n90, Shading Model=ShadingModel()) {attributeSetTypes [#4]}
  ) {group=Shading, parameterName=UsePropMaskForPost}
  n92 = StaticBool()
  n93 = StaticBool() {value=true}
  n94 = StaticSwitchParameter(
   True=NamedRerouteUsage()
   False=MaterialFunctionCall!(
    In=MaterialFunctionCall(Material_In=@n91, UseCrackedDamageTextures=@n92, UseAdditiveWPO=@n93) {materialFunction=&MF_FN_DestructibleDamage}.Material_Out
    DestructibleDamageAmount
   ) {materialFunction=/WildEstateTech/Environment/BaseMaterials/MaterialFunctions/MF_WildEstate_Cracks}.Result
  ) {group=Shading, parameterName=UseAnimeBase, sortPriority=0}
  n95 = GetMaterialAttributes(None=MaterialFunctionCall!(
    Material=MaterialFunctionCall!(In=@n94) {materialFunction=/WildEstateTech/Environment/BaseMaterials/MaterialFunctions/MF_WildEstate_FurSilhouette}.Result
    SFXMask=NamedRerouteUsage()
   ) {materialFunction=/WildEstateTech/Environment/BaseMaterials/MaterialFunctions/MF_WildEstate_Gameplay_TechArt}.Result) {attributeGetTypes [#1]}
  SetMaterialAttributes!(
   MaterialAttributes=@n95.MaterialAttributes
   Normal=MaterialProxyReplace(Realtime=@n95.Normal, MaterialProxy=Transform(None=@n95.Normal) {transformSourceType=TRANSFORMSOURCE_World, transformType=TRANSFORM_Tangent})
  ) {attributeSetTypes [#1]}
 ]
}
