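# Unlit additive material for a simple airglow sprite, drawing a tintable soft-capable disc from a radial UV mask.
# Color times brightness is divided by eye adaptation to counter auto exposure, then the attributes pass through a shared emissive-controls function.
# Opacity adds depth fade to the mask, except at Low and Medium quality, which use the bare mask.
# World position offset combines a scalable billboard function with a camera offset.
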Material /WildEstateTech/Environment/BaseMaterials/M_WildEstate_SimpleAirglow
graph {
 properties {bAutomaticallySetUsageInEditor=false, bUseMaterialAttributes=true, blendMode=BLEND_Additive, shadingModel=MSM_Unlit, usedShadingModels=Unlit}
 nodes [
  n1 = Multiply(
   A=VectorParameter() {group=Color, parameterName=Color, defaultValue=<rgba 1 0 0 1>}.RGB
   B=ScalarParameter() {defaultValue=1, group=Color, parameterName=Brightness}
  )
  n2 = EyeAdaptation()
  n3 = SphereMask(
   A=TextureCoordinate()
   B=Constant() {r=0.5}
   Radius=ScalarParameter() {defaultValue=0.5, group=Opacity, parameterName=Radius}
   Hardness=ScalarParameter() {defaultValue=0.01, group=Opacity, parameterName=Hardness}
  ) {attenuationRadius=0.5, hardnessPercent=1}
  n4 = DepthFade(Opacity=@n3, FadeDistance=ScalarParameter() {defaultValue=15, group=Opacity, parameterName=DepthFade})
  n5 = MaterialFunctionCall!(UVs, Scale=ScalarParameter() {defaultValue=1, group=WPO, parameterName=Scale}, UV Center) {materialFunction=/WildEstateTech/Environment/BaseMaterials/MaterialFunctions/MF_Billboard}
  n6 = MaterialFunctionCall(Offset Amount=ScalarParameter() {defaultValue=10, parameterName=CameraOffset}, World Position=ObjectPositionWS()) {materialFunction=/Engine/Functions/Engine_MaterialFunctions02/WorldPositionOffset/CameraOffset}
  MaterialFunctionCall!(
   SFX
   Material In=MakeMaterialAttributes!(
    BaseColor
    Metallic
    Specular
    Roughness
    Anisotropy
    EmissiveColor=Divide(A=@n1, B=@n2.EyeAdaptation)
    Opacity=QualitySwitch(Default=@n4, Low=@n3, Medium=@n3)
    OpacityMask
    Normal
    Tangent
    WorldPositionOffset=Add(A=@n5.Position, B=@n6.Clamped Camera Offset)
    SubsurfaceColor
    ClearCoat
    ClearCoatRoughness
    AmbientOcclusion
    Refraction
    CustomizedUVs_0
    CustomizedUVs_1
    CustomizedUVs_2
    CustomizedUVs_3
    CustomizedUVs_4
    CustomizedUVs_5
    CustomizedUVs_6
    CustomizedUVs_7
    PixelDepthOffset
    ShadingModel
    Displacement
   )
   SFX Textures RGB
  ) {materialFunction=/Game/Global/Base4Layers_2022/MaterialFunctions/MF_EmissiveControls_2}
 ]
}
